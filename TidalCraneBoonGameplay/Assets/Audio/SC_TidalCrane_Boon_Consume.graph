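# Sound cue for consuming a Tidal Crane boon, reusing a Vivid Razor boons takedown medallion wave.
# A Modulator root applies slight random pitch variation to the single wave player.
# The cue uses weapon pickup attenuation and the loot collect sound class, with a concurrency override limiting it to one playing instance.
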SoundCue /TidalCraneBoonGameplay/Assets/Audio/SC_TidalCrane_Boon_Consume
root=Modulator_0
nodes [
 {id=Modulator_0, children [WavePlayer_0], properties {pitchMax=1.03, pitchMin=0.97}}
 {id=WavePlayer_0, properties {soundWaveAssetPtr=&/VividRazorBoons_TakeDown_Medallion}}
]
properties {
 attenuationSettings=&Fort_Weapon_Pickup_Att
 bOverrideConcurrency=true
 maxDistance=3500
 soundClassObject=&Fort_Loot_Collect
 volumeMultiplier=0.9
 concurrencyOverrides {maxCount=1}
}
