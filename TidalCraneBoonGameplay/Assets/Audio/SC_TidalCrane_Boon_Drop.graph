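# Sound cue for dropping a Tidal Crane boon, reusing the default Takedown Medallions drop sound.
# A single wave player feeds a Modulator left at its usual settings at the root.
# The cue plays at reduced volume through the weapon-drop attenuation settings and the Gadgets LTM sound class.
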SoundCue /TidalCraneBoonGameplay/Assets/Audio/SC_TidalCrane_Boon_Drop
root=Modulator_2
nodes [
 {id=Modulator_2, children [WavePlayer_6]}
 {id=WavePlayer_6, properties {soundWaveAssetPtr=&/Default_TakedownMedallions_Drop}}
]
properties {
 attenuationSettings=&Fort_Weapon_Drop_Att
 maxDistance=3000
 soundClassObject=/Game/Sounds/FortSoundClassesAndModes/Fort_Gadgets_LTM
 volumeMultiplier=0.4
}
