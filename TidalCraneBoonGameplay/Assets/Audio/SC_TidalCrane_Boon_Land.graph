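# SoundCue for the landing sound of the Tidal Crane boon in the TidalCraneBoonGameplay plugin.
# A mixer layers a jetpack drop wave with a vending machine sparkle wave at equal volume.
# Separate modulators hold fixed pitches, raising the drop and lowering the sparkle.
# The cue plays at reduced volume with weapon-drop attenuation and the gadgets LTM sound class.
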SoundCue /TidalCraneBoonGameplay/Assets/Audio/SC_TidalCrane_Boon_Land
root=Mixer_0
nodes [
 {id=Mixer_0, children [Modulator_0,Modulator_1], properties {inputVolume [1,1]}}
 {id=Modulator_0, children [WavePlayer_2], properties {pitchMax=1.5, pitchMin=1.5}}
 {id=Modulator_1, children [WavePlayer_5], properties {pitchMax=0.8, pitchMin=0.8}}
 {id=WavePlayer_2, properties {soundWaveAssetPtr=&/Jetpack_Drop_01}}
 {id=WavePlayer_5, properties {soundWaveAssetPtr=&/VendingMachine_Vend_Sparkle}}
]
properties {
 attenuationSettings=&Fort_Weapon_Drop_Att
 maxDistance=3000
 soundClassObject=/Game/Sounds/FortSoundClassesAndModes/Fort_Gadgets_LTM
 volumeMultiplier=0.4
}
