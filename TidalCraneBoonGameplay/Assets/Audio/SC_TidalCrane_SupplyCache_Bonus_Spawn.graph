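# Sound cue for the Tidal Crane supply cache bonus spawn, playing a low cash-drop fanfare at reduced volume.
# A single wave player feeds a Modulator node at the root, using the loot collect sound class and a high priority.
# It references the shared weapon drop attenuation and also stores falloff and stereo spread override values.
# Its concurrency override values specify one instance and prevent new plays, with an empty concurrency set entry.
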SoundCue /TidalCraneBoonGameplay/Assets/Audio/SC_TidalCrane_SupplyCache_Bonus_Spawn
root=Modulator_1
nodes [
 {id=Modulator_1, children [WavePlayer_2]}
 {id=WavePlayer_2, properties {soundWaveAssetPtr=&/Cash_Drop_Low_Fanfare}}
]
properties {
 attenuationSettings=&Fort_Weapon_Drop_Att
 maxDistance=3000
 priority=0.9
 soundClassObject=&Fort_Loot_Collect
 volumeMultiplier=0.35
 attenuationOverrides {falloffDistance=1000, stereoSpread=512, attenuationShapeExtents {x=0}}
 concurrencyOverrides {maxCount=1, resolutionRule=PreventNew}
 concurrencySet [null]
}
